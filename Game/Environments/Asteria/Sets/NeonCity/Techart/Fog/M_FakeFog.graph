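# Translucent lit fake fog material for NeonCity meshes, faded by time of day.
# Cloud noise blends a bounding-box top projection with side UVs by a V gradient, using flowmap warping and reflection-vector parallax.
# Density combines noise with depth fade, a view-facing edge falloff and camera-distance fade.
# An offset toward the actor position collapses the mesh when the time-of-day alpha is near zero, and always on mobile.
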Material /Game/Environments/Asteria/Sets/NeonCity/Techart/Fog/M_FakeFog
graph {
 properties {bCastRayTracedShadows=false, blendMode=BLEND_Translucent, pixelDepthOffsetMode=PDOM_Legacy, usedShadingModels=Default Lit}
 nodes [
  Constant() {r=0.5}
  n1 = ScalarParameter() {defaultValue=2000, parameterName=Fade}
  n2 = CameraVectorWS()
  n3 = VertexNormalWS()
  n4 = Saturate(None=Multiply(A=DotProduct(A=@n2, B=@n3), B=ScalarParameter() {defaultValue=1, parameterName=flip}))
  n5 = Multiply(
   A=Multiply(A=NamedRerouteUsage(), B=DepthFade(FadeDistance=@n1) {fadeDistanceDefault=800})
   B=Power(
    Base=MaterialFunctionCall(Input=@n4) {materialFunction=&RemapValueRange}.Result
    Exp=ScalarParameter() {defaultValue=5, parameterName=Pow}
   )
  )
  n6 = OneMinus(None=Power(Base=OneMinus(None=@n5)) {constExponent=4})
  n7 = ScalarParameter() {defaultValue=0.35, parameterName=density}
  n8 = MaterialFunctionCall(Input=PixelDepth().R, Input Low=Constant() {r=1000}, Input High=Constant() {r=6500}) {materialFunction=&RemapValueRange}
  n9 = CollectionParameter() {
   collection=&FortniteMaterialParameters
   parameterName=FortniteActiveTimeOfDay
  }
  Multiply(A=Multiply(A=Multiply(A=@n6, B=@n7), B=Saturate(None=@n8.Result)), B=ComponentMask(None=@n9) {a=true})
  n10 = Multiply(A=Time()) {constB=0.1}
  n11 = Multiply(A=ConstantBiasScale(None=MaterialFunctionCall(Value=@n10) {materialFunction=&LinearSine}."Rounded Linear Sine ") {bias=-0.5, scale=2}) {constB=0.01}
  NamedRerouteDeclaration(None=@n11) {name=warp offset, nodeColor=<rgba 0.8984 0 1 1>}
  n12 = Multiply(A=ComponentMask(None=MaterialFunctionCall() {materialFunction=&BoundingBoxBased_0-1_UVW}.RGB) {g=true, r=true}) {constB=0.5}
  n13 = TextureSample(UVs=@n12) {samplerType=SAMPLERTYPE_LinearColor, texture=/Game/Global/Textures/Default/Flowmaps/T_SwirlFlow_02}
  n14 = Multiply(A=ComponentMask(None=@n13.RGB) {g=true, r=true}, B=NamedRerouteUsage())
  NamedRerouteDeclaration(None=TextureSample(UVs=Subtract(A=@n12, B=@n14)) {texture=&AtmosphericCloudNoise01}.RGB) {name=top projection, nodeColor=<rgba 0.3828 0 1 1>}
  n15 = TextureCoordinate()
  n16 = ComponentMask(None=TextureSample(UVs=Multiply(A=@n15) {constB=4}) {samplerType=SAMPLERTYPE_LinearColor, texture=&T_SwirlFlow_02}.RGB) {g=true, r=true}
  n17 = Time()
  n18 = Subtract(
   A=@n15
   B=MaterialFunctionCall(UVs=Multiply(A=@n16, B=@n11), "Rotation Angle (0-1)"=Multiply(A=@n17) {constB=0.0001}) {materialFunction=&CustomRotator}.Rotated Values
  )
  n19 = MaterialFunctionCall!(
   Normal=Constant3Vector() {constant=<rgba 0 0 1 0>}
   CameraVector=Transform(None=CameraVectorWS()) {transformSourceType=TRANSFORMSOURCE_World, transformType=TRANSFORM_Tangent}
  ) {materialFunction=/Engine/Functions/Engine_MaterialFunctions02/Utility/CustomReflectionVector}
  n20 = Multiply(
   A=Divide(A=ComponentMask(None=@n19.Reflection Vector) {g=true, r=true}, B=ComponentMask(None=@n19.Reflection Vector) {b=true})
   B=ScalarParameter() {defaultValue=0.15, parameterName=bump}
  )
  n21 = Add(
   A=Multiply(A=Add(A=@n18, B=@n20), B=ScalarParameter() {defaultValue=0.45, parameterName=tile}) {constB=0.6666}
   B=Constant2Vector() {g=0.3333, r=0.3333}
  )
  NamedRerouteDeclaration(None=Max(
    A=TextureSample(UVs=@n21) {texture=&AtmosphericCloudNoise01}.RGB
    B=TextureSample(UVs=@n18) {texture=&AtmosphericCloudNoise01}.RGB
   )) {name=side mapped, nodeColor=<rgba 1 0 0.7969 1>}
  n22 = ComponentMask(None=TextureCoordinate()) {g=true}
  n23 = Abs(None=Multiply(A=Subtract(A=@n22) {constB=0.5}) {constB=2})
  NamedRerouteDeclaration(None=Power(Base=OneMinus(None=@n23), Exp=ScalarParameter() {defaultValue=1, parameterName=PowTop})) {name=topsidelerp, nodeColor=<rgba 1 0.2812 0 1>}
  NamedRerouteDeclaration(None=LinearInterpolate(A=NamedRerouteUsage(), B=NamedRerouteUsage(), Alpha=NamedRerouteUsage())) {name=lerpedtexture, nodeColor=<rgba 1 0.8906 0 1>}
  n24 = CollectionParameter() {
   collection=&FortniteMaterialParameters
   parameterName=FortniteActiveTimeOfDay
  }
  n25 = Constant4Vector() {constant=<rgba 0 0 0 1>}
  n26 = Subtract(A=ActorPositionWS(), B=WorldPosition().XYZ)
  ShadingPathSwitch(Default=If(A=DotProduct(A=@n24, B=@n25.RGBA), "A > B"=Constant(), "A < B"=@n26) {constB=0.01}, Mobile=@n26) {desc=Migrated from FeatureLevelSwitch}
  Subtract(A=@n13.RGB) {constB=0.5}
  TextureCoordinate()
  TextureSample(UVs=Multiply(A=TextureCoordinate()) {constB=2}) {texture=&AtmosphericCloudNoise01}
  VertexNormalWS()
 ]
}
