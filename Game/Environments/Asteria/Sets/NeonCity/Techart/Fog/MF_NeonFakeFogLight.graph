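# Material function that fakes a glowing fog light on translucent spheres in the NeonCity set.
# Opacity sums scene surface lighting from depth-reconstructed positions (normal-weighted on deferred) and a ray-traced sphere falloff, with distance and depth fades.
# A static bool selects near or far sphere fading and culling, which collapses vertices onto the camera through WPO.
# Emissive blends two colors, scaled down on mobile and forward, adding scene color at Epic quality.
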MaterialFunction /Game/Environments/Asteria/Sets/NeonCity/Techart/Fog/MF_NeonFakeFogLight
graph {nodes [
  n1 = NamedRerouteUsage()
  n2 = ScalarParameter() {defaultValue=15000, parameterName=CameraFadeOffset}
  n3 = Clamp(None=Divide(A=Subtract(A=@n1, B=@n2), B=ScalarParameter() {defaultValue=10000, parameterName=CameraFadeLength}))
  n4 = Distance(
   A=MaterialFunctionCall() {materialFunction=&WorldPositionBehindTranslucency}.Result
   B=ObjectPositionWS()
  )
  n5 = ScalarParameter() {defaultValue=2706.8237, parameterName=LightRadius}
  n6 = OneMinus(None=Clamp(None=Divide(A=@n4, B=@n5)))
  n7 = SceneTexture() {sceneTextureId=PPI_WorldNormal}
  n8 = Subtract(
   A=ObjectPositionWS()
   B=MaterialFunctionCall() {materialFunction=&WorldPositionBehindTranslucency}.Result
  )
  n9 = LinearInterpolate(
   A=ScalarParameter() {defaultValue=1, parameterName=MinLightTop}
   Alpha=DotProduct(A=ComponentMask(None=@n7.Color) {b=true, g=true, r=true}, B=Normalize(VectorInput=@n8))
  ) {constA=1}
  n10 = Constant() {r=1}
  n11 = ShadingPathSwitch(Default=@n9, Deferred=@n9, Forward=@n10, Mobile=@n10)
  n12 = CameraPositionWS()
  n13 = Subtract(A=WorldPosition().XYZ, B=@n12) {constB=-1}
  n14 = OneMinus(None=MaterialFunctionCall!(
    Sphere Radius=Constant() {r=1}
    Sphere Falloff=ScalarParameter() {defaultValue=2706.8237, parameterName=LightRadius}
    Unit Ray Direction=Normalize(VectorInput=@n13)
    Sphere Center=ObjectPositionWS()
    Ray Origin=@n12
   ) {materialFunction=/Game/Athena/Environments/MaterialFunctions/RayTraceSphereFalloffMinClamp}.Clamped Falloff)
  n15 = ScalarParameter() {defaultValue=2, parameterName=FalloffFog}
  n16 = Clamp(None=Add(A=Multiply(A=@n6, B=@n11), B=Power(Base=@n14, Exp=@n15)))
  n17 = Add(
   A=Multiply(A=ScalarParameter() {defaultValue=0.188, parameterName=FalloffOuterBirghtness}, B=@n16) {constA=0.5}
   B=Power(Base=@n16, Exp=ScalarParameter() {defaultValue=2.7198, parameterName=Falloff})
  )
  n18 = Multiply(A=VertexInterpolator(VS=OneMinus(None=@n3)).PS, B=@n17)
  n19 = Multiply(A=NamedRerouteUsage(), B=@n18)
  n20 = Multiply(A=NamedRerouteUsage(), B=@n18)
  n21 = ScalarParameter() {defaultValue=2000, parameterName=DepthFadeDistance}
  n22 = CameraPositionWS()
  n23 = Distance(A=MaterialFunctionCall() {materialFunction=&WorldPositionBehindTranslucency}.Result, B=@n22)
  n24 = ObjectRadius()
  n25 = Distance(A=@n22, B=ObjectPositionWS())
  n26 = Divide(A=Add(A=@n23, B=Subtract(A=@n24, B=@n25)), B=@n21)
  n27 = Multiply(A=NamedRerouteUsage(), B=@n17)
  n28 = Multiply(A=NamedRerouteUsage(), B=@n17)
  n29 = StaticSwitch(
   True=DepthFade(Opacity=StaticSwitchParameter(True=@n19, False=@n20) {defaultValue=true, parameterName=UsePPVFade}, FadeDistance=@n21)
   False=Multiply(A=Saturate(None=@n26), B=StaticSwitchParameter(True=@n27, False=@n28) {defaultValue=true, parameterName=UsePPVFade})
   Value=NamedRerouteUsage()
  )
  n30 = NamedRerouteUsage()
  n31 = Subtract(A=CameraPositionWS(), B=WorldPosition().XYZ)
  n32 = NamedRerouteUsage()
  n33 = ScalarParameter() {defaultValue=3579.1655, parameterName=SphereRadius}
  n34 = Ceil(None=Saturate(None=Subtract(A=@n32, B=@n33)))
  n35 = Clamp(None=Add(A=Floor(None=@n3), B=OneMinus(None=@n34)))
  n36 = NamedRerouteUsage()
  n37 = NamedRerouteUsage()
  n38 = Multiply(A=@n37, B=NamedRerouteUsage())
  n39 = Saturate(None=Ceil(None=StaticSwitchParameter(True=@n38, False=@n37) {defaultValue=true, parameterName=UsePPVFade}))
  n40 = Saturate(None=Add(A=StaticSwitch(True=@n35, False=@n34, Value=@n36), B=OneMinus(None=@n39)))
  n41 = SceneColor!(UVs)
  n42 = ScalarParameter() {defaultValue=0.5, parameterName=Blend_SceneColor_Final}
  n43 = Subtract(A=NamedRerouteUsage(), B=ScalarParameter() {parameterName=PPV_Transition_Offset})
  n44 = Divide(A=@n43, B=ScalarParameter() {defaultValue=5000, parameterName=Color_Transition_Length})
  n45 = ScalarParameter() {defaultValue=5, parameterName=Brightness}
  n46 = Multiply(A=@n45, B=ScalarParameter() {defaultValue=0.7, parameterName=BrightnessMobileScale}) {constB=0.4}
  n47 = Multiply(
   A=LinearInterpolate(
    A=VectorParameter() {parameterName=Color_Near, defaultValue=<rgba 1.2067 1.6933 2 1>}.RGB
    B=VectorParameter() {parameterName=Color, defaultValue=<rgba 1 0.168 0.0312 1>}.RGB
    Alpha=Saturate(None=@n44)
   )
   B=ShadingPathSwitch(Default=@n45, Deferred=@n45, Forward=@n46, Mobile=@n46)
  )
  n48 = QualitySwitch(Default=@n47, Epic=Add(A=@n47, B=Multiply(A=@n41.RGB, B=@n42)))
  FunctionOutput(None=MakeMaterialAttributes!(
    BaseColor
    Metallic
    Specular
    Roughness
    Anisotropy
    EmissiveColor=ShadingPathSwitch(Default=@n48, Mobile=@n47) {desc=Migrated from FeatureLevelSwitch}
    Opacity=Multiply(A=@n29, B=@n30)
    OpacityMask
    Normal
    Tangent
    WorldPositionOffset=Multiply(A=@n31, B=@n40)
    SubsurfaceColor
    ClearCoat
    ClearCoatRoughness
    AmbientOcclusion
    Refraction
    CustomizedUVs_0
    CustomizedUVs_1
    CustomizedUVs_2
    CustomizedUVs_3
    CustomizedUVs_4
    CustomizedUVs_5
    CustomizedUVs_6
    CustomizedUVs_7
    PixelDepthOffset
    ShadingModel
    Displacement
   ))
  NamedRerouteDeclaration(None=Distance(A=ObjectPositionWS(), B=CameraPositionWS())) {name=Distance, nodeColor=<rgba 0.8516 1 0 1>}
  n49 = VectorParameter() {group=Emissive Autoexposure, parameterName=Emissive multiplier over day phases, defaultValue=<rgba 1 1 1 1>}
  NamedRerouteDeclaration(None=MaterialFunctionCall(Value_Morning=@n49.R, Value_Day=@n49.R, Value_Evening=@n49.R, Value_Night=@n49.R) {materialFunction=&Scalar_Time_of_Day}.Result) {name=TODM Mask, nodeColor=<rgba 0 1 0.25 1>}
  n50 = FunctionInput(Preview=StaticBool()) {bUsePreviewValueAsDefault=true, inputType=FunctionInput_StaticBool}
  NamedRerouteDeclaration(None=@n50) {name=Is Far Sphere, nodeColor=<rgba 0.8516 1 0 1>}
  n51 = ScalarParameter() {defaultValue=2000, parameterName=NearFadeDistance}
  NamedRerouteDeclaration(None=Saturate(None=Divide(A=@n23, B=@n51))) {name=NearFade, nodeColor=<rgba 0 0.5312 1 1>}
  n52 = ScalarParameter() {defaultValue=1, parameterName=PPV_Transition_Length}
  NamedRerouteDeclaration(None=Saturate(None=Divide(A=@n43, B=@n52))) {name=PPV Fade, nodeColor=<rgba 1 0 0.7266 1>}
  Reroute(None=@n50)
 ]}
