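# Additive unlit material for faked fog lights in the NeonCity environment, named as the far variant.
# It feeds a static bool parameter, on by default, into the shared MF_NeonFakeFogLight function, which holds the fog light logic.
# That result passes through the global MF_LockdownPulse function to drive the material attributes.
# It renders before depth of field with translucency vertex fog disabled.
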Material /Game/Environments/Asteria/Sets/NeonCity/Techart/Fog/M_Neon_FakeFogLightFar
graph {
 properties {
  bUseMaterialAttributes=true
  bUseTranslucencyVertexFog=false
  blendMode=BLEND_Additive
  pixelDepthOffsetMode=PDOM_Legacy
  shadingModel=MSM_Unlit
  translucencyPass=MTP_BeforeDOF
  usedShadingModels=Unlit
 }
 nodes [MaterialFunctionCall!(MA In=MaterialFunctionCall!(In=StaticBoolParameter() {defaultValue=true}) {materialFunction=&/MF_NeonFakeFogLight}.Result) {materialFunction=/Game/Global/MaterialFunctions/MF_LockdownPulse}]
}
